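# Physics asset for the Drum ShooFly gadget prop in BRCosmetics, giving its skeletal mesh a simple collision body.
# A single capsule on the root bone encloses the prop; its radius far exceeds its length, so it is almost a sphere.
# It has no constraints or other bodies, so it defines no articulated parts.
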PhysicsAsset /BRCosmetics/Animation/Props/Gadgets_Drum_ShooFly/Meshes/PA_Drum_ShooFly_Gadget
bodies {root [{shapeType=Capsule, length=2.6182, radius=153.8589, center=<0 -0.1024 1.4467>, rotation=<rot 2.66 -90 -0.0014>}]}
constraints []
